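# Editor utility widget that wraps a GraphClass renderer, created on Construct, with editable X/Y min/max fields and a zoom selection overlay.
# SetGraphData forwards points to the renderer and broadcasts OnDataChanged.
# Zoom commits set the X range; typed limits are clamped and order-checked, then both update the renderer's display area, choosing auto-size by comparison with its current range.
# Renderer range changes refresh the fields without firing events, avoiding feedback loops.
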EditorUtilityWidgetBlueprint /Game/NeverCook/EditorUtilities/Generic/GraphWidgets/GraphDisplay/EUW_GraphDisplay
graphs [
 CreateGraphWidget {nodes [
   CreateGraphWidget() {is=title K2Node_FunctionEntry}
   > n1 = Sequence!(execute) -> (then_0, then_1, then_2:Exec, then_3:Exec)
   > ClearChildren(self=Get OverlayGraphRenderer() -> (OverlayGraphRenderer:Overlay Object Reference))
   n2 = Create Widget!(
    execute:Exec=@n1.then_1
    WorldContextObject:Object Reference
    Class:User Widget Class Reference=Get GraphClass() -> (GraphClass:EUW Graph Renderer Class Reference)
    OwningPlayer:Player Controller Object Reference
   ) -> (then:Exec, ReturnValue:EUW Graph Renderer Object Reference)
   > n3 = Set GraphRenderer(GraphRenderer:EUW Graph Renderer Object Reference=@n2) -> (Output_Get:EUW Graph Renderer Object Reference)
   > n4 = AddChildToOverlay!(execute, self:Overlay Object Reference=Get OverlayGraphRenderer(), Content:Widget Object Reference=@n3) -> (then, ReturnValue:Overlay Slot Object Reference)
   > SetHorizontalAlignment(self=@n4, InHorizontalAlignment=HAlign_Fill)
   > SetVerticalAlignment(self=@n4, InVerticalAlignment=VAlign_Fill)
   SignupForGraphRendererEvents(execute=@n1.then_2)
   Is Valid(exec=@n1.then_3, InputObject=Get GraphRenderer() -> (GraphRenderer:EUW Graph Renderer Object Reference))
   > Call On Graph Created and Ready!(execute:Exec, self:EUW Graph Display Object Reference, GraphDisplay:EUW Graph Display Object Reference=Self-Reference()) -> (then:Exec) {is=title K2Node_CallDelegate}
  ]}
 EventGraph {nodes [
   n1 = "On Zoom Committed (EUW_GraphDisplaySelectionOverlay)"!() -> (OutputDelegate, then, StartEndPercent:Vector 2D Structure) {is=title K2Node_ComponentBoundEvent}
   n2 = Get GraphRenderer()
   > Is Valid(InputObject=@n2)
   n3 = BreakVector2D(InVec=@n1.StartEndPercent)
   n4 = Get!(self:EUW Graph Renderer Object Reference=@n2) -> (DisplayMinMaxX_X:double, DisplayMinMaxX_Y:double)
   > SetValue!(
    execute
    self:EUW Line Graph Value Display Object Reference=Get ValueX_Min() -> (ValueX_Min:EUW Line Graph Value Display Object Reference)
    Value:double=Lerp(A=@n4, B=@n4.DisplayMinMaxX_Y, Alpha=@n3)
    ThrowEvent:bool=false
    ForceSet:bool=false
   ) -> (then)
   > SetValue(
    self=Get ValueX_Max() -> (ValueX_Max:EUW Line Graph Value Display Object Reference)
    Value=Lerp(A=@n4, B=@n4.DisplayMinMaxX_Y, Alpha=@n3.Y)
    ThrowEvent=false
    ForceSet=false
   )
   > ValidateGraphRendererMinMax()
   n5 = "On Value Changed (ValueY_Min)"!() -> (OutputDelegate, then, OldValue:double, NewValue:double, Widget:EUW Line Graph Value Display Object Reference) {is=title K2Node_ComponentBoundEvent}
   > ValidateValueText_MinY!(
    execute
    self
    OldValue:double=@n5.OldValue
    NewValue:double=@n5.NewValue
    IsX:bool=false
    isMin:bool=true
    InWidget:EUW Line Graph Value Display Object Reference=@n5.Widget
    OtherWidget:EUW Line Graph Value Display Object Reference=Get ValueY_Max() -> (ValueY_Max:EUW Line Graph Value Display Object Reference)
   ) -> (then)
   n6 = "On Value Changed (ValueY_Max)"!() -> (OutputDelegate, then, OldValue:double, NewValue:double, Widget:EUW Line Graph Value Display Object Reference) {is=title K2Node_ComponentBoundEvent}
   > ValidateValueText_MinY(
    OldValue=@n6.OldValue
    NewValue=@n6.NewValue
    IsX=false
    isMin=false
    InWidget=@n6.Widget
    OtherWidget=Get ValueY_Min() -> (ValueY_Min:EUW Line Graph Value Display Object Reference)
   )
   n7 = "On Value Changed (ValueX_Min)"!() -> (OutputDelegate, then, OldValue:double, NewValue:double, Widget:EUW Line Graph Value Display Object Reference) {is=title K2Node_ComponentBoundEvent}
   > ValidateValueText_MinY(OldValue=@n7.OldValue, NewValue=@n7.NewValue, IsX=true, isMin=true, InWidget=@n7.Widget, OtherWidget=Get ValueX_Max())
   n8 = "On Value Changed (ValueX_Max)"!() -> (OutputDelegate, then, OldValue:double, NewValue:double, Widget:EUW Line Graph Value Display Object Reference) {is=title K2Node_ComponentBoundEvent}
   > ValidateValueText_MinY(OldValue=@n8.OldValue, NewValue=@n8.NewValue, IsX=true, isMin=false, InWidget=@n8.Widget, OtherWidget=Get ValueX_Min())
   Event Construct()
   > Branch(Condition=IsValidClass!(Class:Object Class Reference=Get GraphClass()) -> (ReturnValue:bool))
   > CreateGraphWidget() {is=title K2Node_CallFunction}
  ]}
 GetGraphValueFromPercentage {nodes [
   n1 = GetGraphValueFromPercentage!() -> (then, PercentX:double=~, FromTotal:bool=~) {is=title K2Node_FunctionEntry}
   n2 = Get GraphRenderer()
   > Is Valid(InputObject=@n2)
   > n3 = GetGraphValueFromPercentage!(execute, self:EUW Graph Renderer Object Reference=@n2, PercentX:double=@n1, FromTotal:bool=@n1.FromTotal) -> (then, Point:Vector 2D Structure, PercentValues:Vector 2D Structure) {is=title K2Node_CallFunction}
   > Return Node!(execute, Point:Vector 2D Structure=@n3, PercentValues:Vector 2D Structure=@n3.PercentValues)
  ]}
 HandleGraphEvent_MinMaxChanged {nodes [
   HandleGraphEvent_MinMaxChanged!() -> (then, GraphRenderer:EUW Graph Renderer Object Reference) {is=title K2Node_FunctionEntry}
   > n1 = Sequence!(execute) -> (then_0, then_1)
   n2 = Get(self=Get GraphRenderer())
   > SetValue(self=Get ValueX_Min(), Value=@n2, ThrowEvent=false, ForceSet=true)
   > SetValue(self=Get ValueX_Max(), Value=@n2.DisplayMinMaxX_Y, ThrowEvent=false, ForceSet=true)
   n3 = Get!(self=Get GraphRenderer()) -> (DisplayMinMaxY_X:double, DisplayMinMaxY_Y:double)
   SetValue(execute=@n1.then_1, self=Get ValueY_Min(), Value=@n3, ThrowEvent=false, ForceSet=true)
   > SetValue(self=Get ValueY_Max(), Value=@n3.DisplayMinMaxY_Y, ThrowEvent=false, ForceSet=true)
  ]}
 OnDataChanged {nodes [OnDataChanged!() -> (then, GraphDisplay:EUW Graph Display Object Reference) {is=title K2Node_FunctionEntry}]}
 OnGraphCreatedAndReady {nodes [OnGraphCreatedAndReady!() -> (then, GraphDisplay:EUW Graph Display Object Reference) {is=title K2Node_FunctionEntry}]}
 SetGraphData {nodes [
   n1 = SetGraphData!() -> (then, GraphData:Array of Graph Renderer Point Data Structures) {is=title K2Node_FunctionEntry}
   n2 = Get GraphRenderer()
   > Is Valid(InputObject=@n2)
   > SetGraphData!(execute, self:EUW Graph Renderer Object Reference=@n2, GraphData:Array of Graph Renderer Point Data Structures=@n1) -> (then) {is=title K2Node_CallFunction}
   > Call On Data Changed!(execute:Exec, self:EUW Graph Display Object Reference, GraphDisplay:EUW Graph Display Object Reference=Self-Reference()) -> (then:Exec) {is=title K2Node_CallDelegate}
  ]}
 SignupForGraphRendererEvents {nodes [
   SignupForGraphRendererEvents() {is=title K2Node_FunctionEntry}
   > Is Valid(InputObject=Get GraphRenderer())
   > Bind Event to On Min Max Updated(self:EUW Graph Renderer Object Reference=Get GraphRenderer(), Delegate=Create Event()) {is=title K2Node_AddDelegate}
  ]}
 ValidateGraphRendererMinMax {nodes [
   ValidateGraphRendererMinMax() {is=title K2Node_FunctionEntry}
   > Is Valid(InputObject=Get GraphRenderer())
   n1 = Get GraphRenderer()
   > Set AutoSizeX!(execute, AutoSizeX:bool=false, self:EUW Graph Renderer Object Reference=@n1) -> (then, Output_Get:bool)
   n2 = MakeVector2D(X=GetValue!(self:EUW Line Graph Value Display Object Reference=Get ValueX_Min()) -> (Value:double), Y=GetValue(self=Get ValueX_Max()))
   n3 = Get GraphRenderer()
   n4 = MakeVector2D(X=GetValue(self=Get ValueY_Min()), Y=GetValue(self=Get ValueY_Max()))
   n5 = GetCurrentMinMaxY!(self:EUW Graph Renderer Object Reference=Get GraphRenderer()) -> (CurrentMinMaxY:Vector 2D Structure)
   > SetDisplayArea!(
    execute
    self:EUW Graph Renderer Object Reference=@n1
    NewAutoSizeX:bool="Equal (Vector2D)"!(
     A:Vector 2D Structure=GetCurrentMinMaxX!(self:EUW Graph Renderer Object Reference=@n3) -> (CurrentMinMaxX:Vector 2D Structure)
     B:Vector 2D Structure=@n2
     ErrorTolerance:float
    ) -> (ReturnValue:bool=~) {is=title K2Node_PromotableOperator}
    NewDisplayMinMaxX:Vector 2D Structure=@n2
    NewAutoSizeY:Enum_GraphRenderAutosize Enum=Select(
     Option 0:Enum_GraphRenderAutosize Enum=NewEnumerator2
     Option 1:Enum_GraphRenderAutosize Enum=NewEnumerator1
     Index="Equal (Vector2D)"(A=@n5, B=@n4) -> (ReturnValue=~)
    ) -> (ReturnValue:Enum_GraphRenderAutosize Enum)
    NewDisplayMinMaxY:Vector 2D Structure=@n4
   ) -> (then)
  ]}
 ValidateValueText_MinY {nodes [
   ValidateValueText_MinY!() -> (
    then
    OldValue:double=~
    NewValue:double=~
    IsX:bool=~
    isMin:bool=~
    InWidget:EUW Line Graph Value Display Object Reference
    OtherWidget:EUW Line Graph Value Display Object Reference
   ) {is=title K2Node_FunctionEntry}
   > Is Valid(InputObject=Get GraphRenderer())
   n1 = GetCurrentMinMaxY!(self=Get GraphRenderer()) -> (CurrentMinMaxY_X:double, CurrentMinMaxY_Y:double)
   n2 = Get NewValue() -> (NewValue:double)
   n3 = GetCurrentMinMaxX!(self=Get GraphRenderer()) -> (CurrentMinMaxX_X:double, CurrentMinMaxX_Y:double)
   n4 = Get NewValue()
   > n5 = Set l_value(l_value:double=Select(
     Option 0:double="Clamp (Float)"(Value=@n2, Min=@n1, Max=@n1.CurrentMinMaxY_Y)
     Option 1:double="Clamp (Float)"(Value=@n4, Min=@n3, Max=@n3.CurrentMinMaxX_Y)
     Index=Get IsX() -> (IsX:bool)
    ) -> (ReturnValue:double)) -> (Output_Get:double)
   > n6 = Branch(Condition="Not Equal (Float)"(A=@n5, B=Get OldValue() -> (OldValue:double)))
   > n7 = Branch(Condition="Equal (Float)"(A=Get l_value() -> (l_value:double), B=Get NewValue()))
   n8 = GetValue(self=Get OtherWidget() -> (OtherWidget:EUW Line Graph Value Display Object Reference))
   n9 = Get l_value()
   n10 = Branch(
    execute=[@n11.then, @n7.then]
    Condition=Select(
     Option 0:bool="float <= float"(A=@n9, B=@n8) -> (ReturnValue=~)
     Option 1:bool="float >= float"(A=@n9, B=@n8)
     Index=Get isMin() -> (isMin:bool)
    ) -> (ReturnValue:bool)
   )
   > SetValue(self=Get InWidget() -> (InWidget:EUW Line Graph Value Display Object Reference), Value=Get OldValue(), ThrowEvent=false, ForceSet=false)
   ValidateGraphRendererMinMax(execute=@n10.else) {is=title K2Node_CallFunction}
   n11 = SetValue(execute=@n7.else, self=Get InWidget(), Value=Get l_value(), ThrowEvent=false, ForceSet=false)
   SetValue(execute=@n6.else, self=Get InWidget(), Value=Get OldValue(), ThrowEvent=false, ForceSet=false)
  ]}
]
variables [
 GraphClass {type=Class, default=null}
 GraphRenderer {type=EUW_GraphRenderer_C, default=null}
 OnDataChanged {default=()}
 OnGraphCreatedAndReady {default=()}
]
widgets {
 info {rootWidgetClass=&StackBox, widgetCount=16}
 widgets [
  StackBox_0 {widgetClassPath=&StackBox, widget=StackBox_0}
  Border_0 {
   widgetClassPath=&Border
   parent=StackBox_0
   widget=Border_0
   properties {brushColor=<rgba 0.0208 0.0208 0.0208 1>}
   slot {class=StackBoxSlot, properties {size {sizeRule=Fill, value=1}}}
  }
  HorizontalBox_0 {widgetClassPath=&HorizontalBox, parent=Border_0, widget=HorizontalBox_0, slot {class=BorderSlot}}
  VerticalBox_0 {
   widgetClassPath=&VerticalBox
   parent=HorizontalBox_0
   widget=VerticalBox_0
   slot {class=HorizontalBoxSlot, properties {size {sizeRule=Fill, value=1}}}
  }
  Overlay_1 {
   widgetClassPath=&Overlay
   parent=VerticalBox_0
   widget=Overlay_1
   slot {class=VerticalBoxSlot, properties {size {sizeRule=Fill, value=1}}}
  }
  OverlayGraphRenderer {
   widgetClassPath=&Overlay
   bIsVariable=true
   parent=Overlay_1
   widget=OverlayGraphRenderer
   slot {class=OverlaySlot, properties {horizontalAlignment=HAlign_Fill, verticalAlignment=VAlign_Fill}}
  }
  EUW_GraphDisplaySelectionOverlay {
   widgetClassPath=/Game/NeverCook/EditorUtilities/Generic/GraphWidgets/GraphDisplay/SubWidgets/EUW_GraphDisplaySelectionOverlay
   bIsVariable=true
   parent=Overlay_1
   widget=EUW_GraphDisplaySelectionOverlay
   slot {class=OverlaySlot, properties {horizontalAlignment=HAlign_Fill, verticalAlignment=VAlign_Fill}}
  }
  HorizontalBox_1 {
   widgetClassPath=&HorizontalBox
   parent=VerticalBox_0
   widget=HorizontalBox_1
   slot {class=VerticalBoxSlot, properties {verticalAlignment=VAlign_Top}}
  }
  ValueX_Min {
   widgetClassPath=/Game/NeverCook/EditorUtilities/Generic/GraphWidgets/GraphDisplay/SubWidgets/EUW_LineGraphValueDisplay
   bIsVariable=true
   parent=HorizontalBox_1
   widget=ValueX_Min
   properties {isMin=true, isX=true}
   slot {class=HorizontalBoxSlot}
  }
  EUW_LineGraphSpan {
   widgetClassPath=/Game/NeverCook/EditorUtilities/Generic/GraphWidgets/GraphDisplay/SubWidgets/EUW_LineGraphSpan
   bIsVariable=true
   parent=HorizontalBox_1
   widget=EUW_LineGraphSpan
   properties {isX=true}
   slot {class=HorizontalBoxSlot, properties {size {sizeRule=Fill, value=1}}}
  }
  ValueX_Max {
   widgetClassPath=&EUW_LineGraphValueDisplay
   bIsVariable=true
   parent=HorizontalBox_1
   widget=ValueX_Max
   properties {isX=true}
   slot {class=HorizontalBoxSlot, properties {horizontalAlignment=HAlign_Right}}
  }
  SizeBox_0 {
   widgetClassPath=&SizeBox
   parent=HorizontalBox_0
   widget=SizeBox_0
   properties {bOverride_MaxDesiredHeight=true, bOverride_MinDesiredWidth=true, maxDesiredHeight=80, minDesiredWidth=50}
   slot {class=HorizontalBoxSlot}
  }
  VerticalBox_1 {
   widgetClassPath=&VerticalBox
   parent=SizeBox_0
   widget=VerticalBox_1
   slot {class=SizeBoxSlot, properties {padding=<ltrb 0 0 0 22>}}
  }
  ValueY_Max {
   widgetClassPath=&EUW_LineGraphValueDisplay
   bIsVariable=true
   parent=VerticalBox_1
   widget=ValueY_Max
   slot {class=VerticalBoxSlot}
  }
  EUW_LineGraphSpan_126 {
   widgetClassPath=&EUW_LineGraphSpan
   bIsVariable=true
   parent=VerticalBox_1
   widget=EUW_LineGraphSpan_126
   slot {class=VerticalBoxSlot, properties {size {sizeRule=Fill, value=1}}}
  }
  ValueY_Min {
   widgetClassPath=&EUW_LineGraphValueDisplay
   bIsVariable=true
   parent=VerticalBox_1
   widget=ValueY_Min
   properties {isMin=true}
   slot {class=VerticalBoxSlot, properties {verticalAlignment=VAlign_Bottom}}
  }
 ]
}
