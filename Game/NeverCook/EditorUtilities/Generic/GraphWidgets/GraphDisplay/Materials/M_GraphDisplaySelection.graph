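# Translucent UI material for the editor-utility GraphDisplay widget that draws a horizontal selection range between StartPercent and EndPercent.
# Sorts the two percents with Min/Max and builds a hard opacity mask along UV X using distance from the midpoint and Ceil.
# Distance to two edge positions blends SelectColor bars into the EndColor fill, in both color and alpha.
# Bar size and fade scale inversely with the UI pixel width to compensate for widget size.
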Material /Game/NeverCook/EditorUtilities/Generic/GraphWidgets/GraphDisplay/Materials/M_GraphDisplaySelection
graph {
 properties {bAutomaticallySetUsageInEditor=false, blendMode=BLEND_Translucent, materialDomain=MD_UI, usedShadingModels=Default Lit}
 nodes [
  n1 = VectorParameter() {parameterName=SelectColor, defaultValue=<rgba 0.6223 0.7057 0.7708 1>}
  n2 = VectorParameter() {parameterName=EndColor, defaultValue=<rgba 0.3646 0.3646 0.3646 0.3>}
  n3 = ComponentMask(None=TextureCoordinate()) {r=true}
  n4 = NamedRerouteUsage()
  n5 = Subtract(A=Distance(A=NamedRerouteUsage(), B=@n3), B=@n4)
  n6 = NamedRerouteUsage()
  n7 = Subtract(A=Distance(A=NamedRerouteUsage(), B=@n3), B=@n4)
  n8 = Min(A=Saturate(None=Divide(A=@n5, B=@n6)), B=Saturate(None=Divide(A=@n7, B=@n6)))
  LinearInterpolate(A=@n1.RGB, B=@n2.RGB, Alpha=@n8)
  Multiply(
   A=LinearInterpolate(A=@n1.A, B=@n2.A, Alpha=@n8)
   B=Multiply(A=NamedRerouteUsage(), B=ScalarParameter() {defaultValue=1, parameterName=Opacity})
  )
  n9 = Add(A=NamedRerouteUsage()) {constB=0}
  NamedRerouteDeclaration(None=Subtract(A=NamedRerouteUsage(), B=@n9)) {name=StartOpacity_Route, nodeColor=<rgba 1 0 0.6328 1>}
  NamedRerouteDeclaration(None=Add(A=@n9, B=NamedRerouteUsage())) {name=EndOpacity_Route, nodeColor=<rgba 0 0.9531 1 1>}
  n10 = ComponentMask(None=TextureCoordinate()) {r=true}
  n11 = NamedRerouteUsage()
  n12 = NamedRerouteUsage()
  n13 = Divide(A=Add(A=@n11, B=@n12))
  n14 = Subtract(A=@n11, B=@n12)
  n15 = Subtract(A=Saturate(None=Distance(A=@n10, B=@n13)), B=Length(None=Divide(A=@n14)) {is=class MaterialExpressionLength})
  NamedRerouteDeclaration(None=OneMinus(None=Ceil(None=@n15))) {name=Opacity_Route, nodeColor=<rgba 1 0.7266 0 1>}
  n16 = NamedRerouteUsage()
  NamedRerouteDeclaration(None=Multiply(A=ScalarParameter() {defaultValue=0.025, parameterName=EndBarSize}, B=@n16)) {name=EndBarSize_route, nodeColor=<rgba 0 1 0.4844 1>}
  NamedRerouteDeclaration(None=Multiply(A=@n16, B=ScalarParameter() {defaultValue=0.01, parameterName=EndBarFade})) {name=EndBarFade_Route, nodeColor=<rgba 1 0.1641 0 1>}
  n17 = ScalarParameter() {defaultValue=0.25, parameterName=StartPercent}
  n18 = ScalarParameter() {defaultValue=0.75, parameterName=EndPercent}
  NamedRerouteDeclaration(None=Min(A=@n17, B=@n18)) {name=Start_Route, nodeColor=<rgba 0.8984 1 0 1>}
  NamedRerouteDeclaration(None=Max(A=@n17, B=@n18)) {name=End_Route, nodeColor=<rgba 0.7813 0 1 1>}
  n19 = MaterialFunctionCall() {materialFunction=&GetUserInterfaceUV}
  NamedRerouteDeclaration(None=Divide(A=Constant() {r=100}, B=ComponentMask(None=@n19.Pixel Size) {r=true})) {name=ScreenSizePercent_Route, nodeColor=<rgba 0.4297 1 0 1>}
 ]
}
